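# Physics asset for the StylusFluff two-handed pickaxe cosmetic, defining collision bodies and joints for the axe and its dangling strap.
# One large capsule covers the handle, and small capsules cover the strap master bone and its two segments.
# Two constraints chain the segments to the master bone, with tight twist and first-swing limits and a wider second swing, so the strap bends mostly in one plane.
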PhysicsAsset /BRCosmetics/Weapons/FORT_Melee/Pickaxe_StylusFluff/Meshes/StylusFluff_Axe_Physics
bodies {
 Melee_TwoHanded_Handle [{shapeType=Capsule, length=50, radius=20, center=<0.8129 2.2263 21.1927>, rotation=<rot 2.8191 -90 0>}]
 dyn_strap_01 [{shapeType=Capsule, length=4, radius=2, shapeName=dyn_strap_01_capsule, center=<3.8543 0 0>, rotation=<rot 90 180 179.9999>}]
 dyn_strap_02 [{shapeType=Capsule, length=4, radius=2, shapeName=dyn_strap_02_capsule, center=<3.9404 0 0>, rotation=<rot 90 0 -0.0001>}]
 dyn_strap_MSTR [{shapeType=Capsule, length=1, radius=1, shapeName=dyn_strap_MSTR_capsule}]
}
constraints [
 {bone1Name=dyn_strap_02, bone2Name=dyn_strap_01, swing1LimitDegrees=2, swing2LimitDegrees=40, twistLimitDegrees=2}
 {bone1Name=dyn_strap_01, bone2Name=dyn_strap_MSTR, swing1LimitDegrees=2, swing2LimitDegrees=40, twistLimitDegrees=2}
]
